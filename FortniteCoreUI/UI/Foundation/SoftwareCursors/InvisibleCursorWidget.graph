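# InvisibleCursorWidget is an empty UserWidget Blueprint in FortniteCoreUI's SoftwareCursors folder, named as a software cursor that draws nothing.
# It has no root widget, child widgets, variables or event graph logic, so it renders no visuals and adds no behavior.
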WidgetBlueprint /FortniteCoreUI/UI/Foundation/SoftwareCursors/InvisibleCursorWidget
graphs [EventGraph {links [], nodes []}]
variables []
widgets {info {inheritedWidgetCount=0, parentClass=&UserWidget, rootWidgetClass=null, widgetCount=0}, widgets []}
